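# Sound cue for the reward icon in the Job Board shadow briefing accept UI, part of the shared quests sounds.
# It plays the JobBoard_Accept_RewardIcon wave through a fixed short delay node with no randomization, so the sound lands slightly after the trigger.
# It routes to the non-critical in-game UI sound class.
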SoundCue /Game/Sounds/Fort_UI_Sounds/SharedQuests/JobBoard_ShadowBriefing_Accept_RewardIcon_Cue
root=Delay_0
nodes [
 {id=Delay_0, children [WavePlayer_3], properties {delayMax=0.4, delayMin=0.4}}
 {id=WavePlayer_3, properties {soundWaveAssetPtr=&/JobBoard_Accept_RewardIcon}}
]
properties {maxDistance=2097152, soundClassObject=&Fort_UI_InGame_NonCritical}
