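# UI sound cue that plays the click for selecting a quest in Shared Quests.
# A single wave player feeds the front-end main button click through a modulator with volume variation disabled.
# It runs in the UI sound class with boosted volume and high priority, and references two concurrency assets for front-end menus and Shared Quests to limit overlapping plays.
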SoundCue /Game/Sounds/Fort_UI_Sounds/SharedQuests/SharedQuests_QuestSelect_Cue
root=Modulator_0
nodes [
 {id=Modulator_0, children [WavePlayer_0], properties {volumeMax=1, volumeMin=1}}
 {id=WavePlayer_0, properties {soundWaveAssetPtr=/Game/Sounds/Fort_UI_Sounds/FrontEnd/New_FrontEnd/UI_FN_Button_Click_Main_01}}
]
properties {
 maxDistance=2097152
 priority=0.9
 soundClassObject=&Fort_UI
 volumeMultiplier=1.8
 concurrencyOverrides {maxCount=1, resolutionRule=StopOldest}
 concurrencySet [
  /Game/Sounds/Concurrency_Rules/FrontEnd_UI_Menu_Max_03_Stop_Oldest
  &/SharedQuests_UI_LimitOneStopLowPri_Conc
 ]
}
